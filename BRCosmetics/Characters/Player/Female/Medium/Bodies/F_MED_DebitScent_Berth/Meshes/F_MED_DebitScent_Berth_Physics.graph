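# Physics asset for the F_MED_DebitScent_Berth female medium player body cosmetic, defining collision and joints for two dangling rope pieces.
# Each side has a three-link capsule chain hanging from a zero-size base rope body, joined by constraints with identical narrow swing and twist limits.
# Spheres on the upper spine and pectoral deform bones approximate the torso and chest, apparently as collision volumes for the ropes.
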PhysicsAsset /BRCosmetics/Characters/Player/Female/Medium/Bodies/F_MED_DebitScent_Berth/Meshes/F_MED_DebitScent_Berth_Physics
bodies {
 base_rope_l [{shapeType=Capsule, length=0, radius=0, shapeName=base_rope_l_capsule}]
 base_rope_r [{shapeType=Capsule, length=0, radius=0, shapeName=base_rope_r_capsule}]
 deform_pec_aim_l [{shapeType=Sphere, radius=4.5, shapeName=deform_pec_aim_l_sphere, center=<4.284 1.4017 -2.0798>}]
 deform_pec_aim_r [{shapeType=Sphere, radius=4.5, shapeName=deform_pec_aim_r_sphere, center=<-4.284 -1.4017 2.0798>}]
 deform_pec_l [{shapeType=Sphere, radius=7.5, shapeName=deform_pec_l_sphere, center=<0 -0.4498 0>}]
 deform_pec_r [{shapeType=Sphere, radius=7.5, shapeName=deform_pec_r_sphere, center=<0 0.4498 0>}]
 dyn_rope_01_l [{shapeType=Capsule, length=2, radius=1.3, shapeName=dyn_rope_02_l_capsule, center=<1.1727 0 0>, rotation=<rot 90 0 0>}]
 dyn_rope_01_r [{shapeType=Capsule, length=2, radius=1.3, shapeName=dyn_rope_02_l_capsule, center=<1.1727 0 0>, rotation=<rot 90 0 0>}]
 dyn_rope_02_l [{shapeType=Capsule, length=2, radius=1.3, shapeName=dyn_rope_02_l_capsule, center=<1.1727 0 0>, rotation=<rot 90 0 0>}]
 dyn_rope_02_r [{shapeType=Capsule, length=2, radius=1.3, shapeName=dyn_rope_02_r_capsule, center=<1.2149 0 0>, rotation=<rot 90 0 0>}]
 dyn_rope_03_l [{shapeType=Capsule, length=1.5, radius=1.3, shapeName=dyn_rope_03_l_capsule, center=<1.1727 0 0>, rotation=<rot 90 0 0>}]
 dyn_rope_03_r [{shapeType=Capsule, length=1.5, radius=1.3, shapeName=dyn_rope_03_r_capsule, center=<1.2149 0 0>, rotation=<rot 90 0 0>}]
 spine_03 [{shapeType=Sphere, radius=16.8, shapeName=spine_03_sphere}]
 spine_04 [{shapeType=Sphere, radius=13, shapeName=spine_04_sphere}]
 spine_05 [{shapeType=Sphere, radius=12, shapeName=spine_05_sphere}]
}
constraints [
 {bone1Name=dyn_rope_01_l, bone2Name=base_rope_l, swing1LimitDegrees=20, swing2LimitDegrees=10, twistLimitDegrees=5}
 {bone1Name=dyn_rope_02_l, bone2Name=dyn_rope_01_l, swing1LimitDegrees=20, swing2LimitDegrees=10, twistLimitDegrees=5}
 {bone1Name=dyn_rope_03_l, bone2Name=dyn_rope_02_l, swing1LimitDegrees=20, swing2LimitDegrees=10, twistLimitDegrees=5}
 {bone1Name=dyn_rope_01_r, bone2Name=base_rope_r, swing1LimitDegrees=20, swing2LimitDegrees=10, twistLimitDegrees=5}
 {bone1Name=dyn_rope_02_r, bone2Name=dyn_rope_01_r, swing1LimitDegrees=20, swing2LimitDegrees=10, twistLimitDegrees=5}
 {bone1Name=dyn_rope_03_r, bone2Name=dyn_rope_02_r, swing1LimitDegrees=20, swing2LimitDegrees=10, twistLimitDegrees=5}
]
